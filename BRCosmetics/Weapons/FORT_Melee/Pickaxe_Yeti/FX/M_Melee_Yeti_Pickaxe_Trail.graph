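# Translucent unlit trail material for the Yeti pickaxe, usable on beam trails and Niagara ribbons.
# Panning, noise-distorted water pattern, half-rounded for stylized banding, adds to tinted turbulence masked by noise cutout, then takes particle color.
# MF_Dissolve, driven by a panning mask and dynamic parameter, adds hard-rounded edges switchable in emissive and opacity.
# Sine UV masks soften ribbon borders in opacity, alongside particle alpha and depth fade, which mobile skips.
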Material /BRCosmetics/Weapons/FORT_Melee/Pickaxe_Yeti/FX/M_Melee_Yeti_Pickaxe_Trail
graph {
 properties {
  bUsedWithBeamTrails=true
  bUsedWithNiagaraRibbons=true
  blendMode=BLEND_Translucent
  pixelDepthOffsetMode=PDOM_Legacy
  shadingModel=MSM_Unlit
  usedShadingModels=Unlit
 }
 nodes [
  n1 = ParticleColor()
  n2 = MaterialFunctionCall(XY=TextureCoordinate() {vTiling=0.5}) {materialFunction=&Swizzle}
  n3 = TextureSample(UVs=Panner(Coordinate=TextureCoordinate()) {speedY=0.5}) {samplerType=SAMPLERTYPE_Grayscale, texture=&LowResBlurredNoise}
  n4 = Multiply(A=@n3.R) {constB=0.2}
  n5 = TextureSample!(UVs=Panner(Coordinate=Add(A=@n2.YX, B=@n4)) {speedY=1}, Tex, Bias=Constant() {r=3}, Apply View MipBias) {
   mipValueMode=TMVM_MipBias
   samplerSource=SSM_Wrap_WorldGroupSettings
   texture=/Game/Effects/Fort_Effects/Textures/Water/T_TilingWater_Pattern
  }
  n6 = Multiply(A=@n5.R, B=Constant3Vector() {constant=<rgba 0.0615 0.8602 1 0>})
  n7 = LinearInterpolate(A=Round(None=@n6), B=@n6, Alpha=Constant() {r=0.5})
  n8 = VectorParameter() {group=color gradient, parameterName=Tint_colour, sortPriority=2, defaultValue=<rgba 0.0552 0.7185 0.8875 1>}
  n9 = Panner(Coordinate=Add(A=TextureCoordinate() {vTiling=0.5}, B=Multiply(A=@n3.R) {constB=0.2})) {speedX=1}
  n10 = Multiply(
   A=@n8.RGB
   B=Multiply(A=TextureSample!(UVs=@n9, Tex, Apply View MipBias) {texture=/Game/Global/Textures/Default/Caustics/T_Turbulence_Seamless}.R)
  )
  n11 = VectorParameter() {group=color gradient, parameterName=Dark_colour, sortPriority=3, defaultValue=<rgba 0.0188 0.1364 0.3125 1>}
  n12 = TextureCoordinate()
  n13 = Multiply(A=@n3.R) {constB=0.2}
  n14 = TextureSample(UVs=Panner(Coordinate=Add(A=@n12, B=@n13)) {speedX=0.2}) {texture=/Game/Effects/Fort_Effects/Textures/Misc/T_Shot0030_NoiseCutout}
  n15 = Multiply(A=Multiply(A=Add(A=@n10, B=@n11.RGB), B=Multiply(A=@n14.RGB) {constB=2})) {constB=0.5}
  n16 = Multiply(A=Multiply(A=@n1.RGB, B=Add(A=@n7, B=@n15)))
  n17 = TextureCoordinate()
  n18 = AppendVector(
   A=ScalarParameter() {defaultValue=1.5, parameterName=Dissolve_UTiling}
   B=ScalarParameter() {defaultValue=0.3, parameterName=Dissolve_VTiling}
  )
  n19 = TextureSampleParameter2D(UVs=Panner(Coordinate=Multiply(A=@n17, B=@n18), Time=ScalarParameter() {defaultValue=0.2, parameterName=Dissolve_Offset}) {speedY=0.5}) {parameterName=Dissolve_Texture, samplerType=SAMPLERTYPE_LinearColor, texture=/Game/Effects/Fort_Effects/BluGlow/Violentium/Violentium_Mask}
  n20 = MaterialFunctionCall!(
   Ramp=DynamicParameter() {defaultValue=<rgba 0.8 1 1 1>}.Param1
   Dissolve=Multiply(A=@n19.G, B=ScalarParameter() {defaultValue=1, parameterName=Dissolve_brightness})
   EdgeSharpness
  ) {materialFunction=/Game/Effects/Fort_Effects/Materials/_Functions/MF_Dissolve}
  n21 = Multiply(A=@n20.ClampedResult)
  n22 = Multiply(A=Multiply(A=@n20.Edge), B=ScalarParameter() {defaultValue=10, parameterName=Dissolve_Edges_Intensity})
  n23 = Add(A=@n21, B=Round(None=Multiply(A=@n22)))
  MaterialFunctionCall(Emissive=StaticSwitchParameter(True=Multiply(A=@n16, B=@n23), False=@n16) {defaultValue=true, parameterName=Use Emissive Color dissolve}) {materialFunction=&MF_FX_EmissiveAdjustments}
  n24 = MaterialFunctionCall(Float2=TextureCoordinate() {vTiling=0.99}) {materialFunction=&BreakOutFloat2Components}
  n25 = Power(Base=Sine(None=@n24.R) {period=2})
  n26 = Power(Base=Sine(None=@n24.G) {period=2}) {constExponent=2.5}
  n27 = Multiply(
   A=Multiply(A=@n1.A, B=Multiply(A=@n25, B=@n26))
   B=StaticSwitchParameter(True=@n23, False=@n21) {defaultValue=true, parameterName=Use Dissolve Edges}
  )
  n28 = DepthFade(Opacity=@n27, FadeDistance=ScalarParameter() {defaultValue=30, parameterName=Depth_Fade})
  ShadingPathSwitch(Default=@n28, Deferred=@n28, Mobile=@n27) {desc=Migrated from FeatureLevelSwitch}
 ]
}
