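# Translucent unlit sprite material for water-drop particles in the Yeti pickaxe effects, usable with Niagara and Cascade.
# A normal map drives a soft reflection vector that, offset per particle, samples a reflection texture, tinted by time-of-day colors for emissive.
# Opacity combines that reflection with the drop texture, particle alpha, optional random variation, distance fade and depth fade.
# Low quality substitutes normal-map blue and skips distance fade; mobile skips depth fade.
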Material /BRCosmetics/Weapons/FORT_Melee/Pickaxe_Yeti/FX/M_Water_drop
graph {
 properties {
  bEnableResponsiveAA=true
  bUsedWithNiagaraSprites=true
  bUsedWithParticleSprites=true
  blendMode=BLEND_Translucent
  pixelDepthOffsetMode=PDOM_Legacy
  shadingModel=MSM_Unlit
  usedShadingModels=Unlit
 }
 nodes [
  n1 = Multiply(A=VectorParameter() {parameterName=Color Morning, defaultValue=<rgba 1 0.6506 0.3424 0>}.RGB) {constB=0.3}
  n2 = Multiply(A=VectorParameter() {parameterName=Color Day, defaultValue=<rgba 0.3004 0.4266 0.5 0>}.RGB) {constB=0.3}
  n3 = Multiply(A=VectorParameter() {parameterName=Color Evening, defaultValue=<rgba 1.3 0.857 0.533 0>}.RGB) {constB=0.3}
  n4 = Multiply(A=VectorParameter() {parameterName=Color Night, defaultValue=<rgba 0.4187 0.6488 1 0>}.RGB)
  n5 = TextureSampleParameter2D() {parameterName=Normal, samplerType=SAMPLERTYPE_Normal, texture=/Game/VisualThreatManager/StormVisuals/Textures/T_Water_drop_N_02}
  n6 = MaterialFunctionCall!(Softness=ScalarParameter() {defaultValue=0.8, parameterName=Softness}, Normal=@n5.RGB) {materialFunction=/Engine/Functions/Engine_MaterialFunctions01/Reflections/ReflectionVectorSoft}
  n7 = ParticleRandom()
  n8 = TextureSample(UVs=ComponentMask(None=Add(A=@n6.Result, B=@n7)) {g=true, r=true}) {texture=/Game/Effects/Fort_Effects/Gameplay/Water/Textures/FX_Reflection_03_Diff}
  n9 = ScalarParameter() {defaultValue=0.01, parameterName=Bias}
  n10 = Multiply(A=Multiply(A=Add(A=@n8.RGB, B=@n9), B=ScalarParameter() {defaultValue=4, parameterName=Brightness}))
  n11 = TextureSampleParameter2D() {parameterName=Normal, samplerType=SAMPLERTYPE_Normal, texture=&T_Water_drop_N_02}
  MaterialFunctionCall(Emissive=Multiply(
    A=MaterialFunctionCall(Color_Morning=@n1, Color_Day=@n2, Color_Evening=@n3, Color_Night=@n4) {materialFunction=&Color_Time_of_Day}.Result
    B=QualitySwitch(Default=@n10, Low=@n11.B)
   )) {materialFunction=&MF_FX_EmissiveAdjustments}
  n12 = TextureSample() {texture=/Game/VisualThreatManager/StormVisuals/Textures/T_Water_drop_02}
  n13 = Add(A=@n10, B=Multiply(B=@n12.B) {constA=1})
  n14 = ConstantBiasScale(None=ParticleRandom()) {bias=0.5}
  n15 = QualitySwitch(Default=StaticSwitchParameter(True=Multiply(A=@n13, B=@n14), False=@n13) {defaultValue=true, parameterName=IsGPUParticle}, Low=@n11.B)
  n16 = ParticlePositionWS()
  n17 = ScalarParameter() {defaultValue=200, parameterName=Start Fade Distance}
  n18 = ScalarParameter() {defaultValue=150, parameterName=End Fade Distance}
  n19 = Multiply(
   A=Multiply(A=ParticleColor().A, B=Multiply(A=@n15, B=@n12.A))
   B=QualitySwitch(
    Default=MaterialFunctionCall(Position=@n16, Fade_Start_Distance=@n17, Fade_End_Distance=@n18) {materialFunction=&MF_Distance_SyncedFade_FX}.Result
    Low=Constant() {r=1}
   )
  )
  n20 = DepthFade(Opacity=@n19, FadeDistance=ScalarParameter() {defaultValue=100, parameterName=Depth_Fade})
  ShadingPathSwitch(Default=@n20, Deferred=@n20, Mobile=@n19) {desc=Migrated from FeatureLevelSwitch}
 ]
}
