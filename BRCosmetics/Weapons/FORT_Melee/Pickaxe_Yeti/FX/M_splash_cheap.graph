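# Cheap unlit, two-sided translucent particle material for splash or smoke puff effects on the Yeti pickaxe, usable on sprites and meshes.
# Particle color, optionally with Fresnel and time-of-day tint, drives emissive.
# Opacity multiplies particle alpha by a splash texture or SubUV flipbook red channel, with depth fade skipped on mobile.
# Static switches add gradient or sphere masks, time-of-day alpha, or a zero-opacity “safe open” mode.
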Material /BRCosmetics/Weapons/FORT_Melee/Pickaxe_Yeti/FX/M_splash_cheap
graph {
 properties {
  bUsedWithMeshParticles=true
  bUsedWithNiagaraMeshParticles=true
  bUsedWithNiagaraSprites=true
  bUsedWithParticleSprites=true
  blendMode=BLEND_Translucent
  pixelDepthOffsetMode=PDOM_Legacy
  shadingModel=MSM_Unlit
  twoSided=true
  usedShadingModels=Unlit
 }
 nodes [
  n1 = Multiply(A=Multiply(A=Fresnel() {baseReflectFraction=0, exponent=2}) {constB=17}, B=Constant3Vector() {constant=<rgba 1 0.3162 0.2465 0>})
  n2 = Constant()
  n3 = ParticleColor()
  n4 = StaticSwitchParameter(True=Add(A=ShadingPathSwitch(Default=@n1, Deferred=@n1, Mobile=@n2) {desc=Migrated from FeatureLevelSwitch}, B=@n3.RGB), False=@n3.RGB) {parameterName=Add Fresnel}
  n5 = Multiply(A=@n4, B=VectorParameter() {parameterName=Smoke_color_evening, defaultValue=<rgba 0.651 0.4525 0.4331 1>}.RGB)
  n6 = Multiply(A=@n4, B=VectorParameter() {parameterName=Smoke_color_night, defaultValue=<rgba 0.1716 0.2135 0.349 1>}.RGB)
  MaterialFunctionCall(Emissive=StaticSwitchParameter(
    True=MaterialFunctionCall(Color_Morning=@n5, Color_Day=@n4, Color_Evening=@n5, Color_Night=@n6) {materialFunction=&Color_Time_of_Day}.Result
    False=@n4
   ) {parameterName=Use_smoke_color_time_of_day}) {materialFunction=&MF_FX_EmissiveAdjustments}
  n7 = WorldPosition()
  n8 = ObjectPositionWS()
  n9 = ScalarParameter() {defaultValue=150, parameterName=Radius}
  n10 = ScalarParameter() {parameterName=Hardness}
  n11 = MaterialFunctionCall!(
   Texture=TextureObject() {samplerType=SAMPLERTYPE_LinearColor, texture=/Game/Effects/Fort_Effects/Gameplay/SmokeDustExplosions/Textures/T_SmokePoof_8x8_01}
   UVs
   SubImages=AppendVector(A=ScalarParameter() {defaultValue=8, parameterName=U}, B=ScalarParameter() {defaultValue=8, parameterName=V})
   Frame=DynamicParameter().Param4
  ) {materialFunction=/Engine/Functions/Engine_MaterialFunctions03/Texturing/SubUV_Function}
  n12 = ScalarParameter() {parameterName=Bias_Intensity}
  n13 = Multiply(
   A=StaticSwitchParameter(
    True=ComponentMask(None=@n11.RGB) {r=true}
    False=TextureSampleParameter2D!(UVs, Bias=@n12, Apply View MipBias) {mipValueMode=TMVM_MipBias, parameterName=Texture, texture=/Game/Effects/Fort_Effects/Textures/Water/T_water_splash_04}.R
   ) {parameterName=Use_SubUV_Texture}
   B=@n3.A
  )
  n14 = DepthFade(Opacity=@n13, FadeDistance=ScalarParameter() {defaultValue=100, parameterName=Depth_Fade})
  n15 = StaticSwitchParameter(True=Constant(), False=ShadingPathSwitch(Default=@n14, Deferred=@n14, Mobile=@n13) {desc=Migrated from FeatureLevelSwitch}) {parameterName=use for safe open}
  n16 = MaterialFunctionCall(Float2=TextureCoordinate()) {materialFunction=&BreakOutFloat2Components}
  n17 = OneMinus(None=@n16.G)
  n18 = Multiply(
   A=Power(Base=Multiply(A=@n16.G), Exp=ScalarParameter() {defaultValue=3, parameterName=Mask_Power01})
   B=Power(Base=Multiply(A=@n17), Exp=ScalarParameter() {defaultValue=3, parameterName=Mask_Power02})
  )
  n19 = StaticSwitchParameter(True=Multiply(A=@n15, B=Multiply(A=@n18)), False=@n15) {parameterName=Use another Mask}
  n20 = Multiply(A=@n19, B=ScalarParameter() {defaultValue=1, parameterName=color_time_of_day_MorningtoEvening})
  n21 = ScalarParameter() {defaultValue=0.1, parameterName=color_time_of_day_Night}
  n22 = StaticSwitchParameter(
   True=MaterialFunctionCall(Color_Morning=@n20, Color_Day=@n20, Color_Evening=@n20, Color_Night=Multiply(A=@n19, B=@n21) {constB=0.1}) {materialFunction=&Color_Time_of_Day}.Result
   False=@n19
  ) {parameterName=Use_Alpha_Color_time_of_day}
  StaticSwitchParameter(True=Multiply(A=@n22, B=SphereMask(A=@n7.XYZ, B=@n8, Radius=@n9, Hardness=@n10)), False=@n22) {parameterName=Use_sphere mask}
 ]
}
