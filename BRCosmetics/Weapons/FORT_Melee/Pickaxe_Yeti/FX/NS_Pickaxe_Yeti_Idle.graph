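# Niagara idle effect for the Yeti pickaxe, combining glinting blue stars and falling water drops as sprites.
# Stars spawn on the surface of an offset box, drift down under drag, and twinkle via an oscillating size curve, alpha fade and dynamic material parameters.
# Water_Drops gates its spawn rate on a comparison of owner speed to a threshold, inherits owner velocity and falls under stronger downward acceleration.
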NiagaraSystem /BRCosmetics/Weapons/FORT_Melee/Pickaxe_Yeti/FX/NS_Pickaxe_Yeti_Idle
dependencies {
 usedDataInterfaces [&NiagaraDataInterfaceCurve]
 usedDynamicInputs [&Vector2DFromFloat]
 usedModules [&SystemState]
}
emitters {
 Stars {
  emitterUpdateScript [
   &EmitterState(Life Cycle Mode=::NewEnumerator1, Loop Duration=0.3, UseLoopDelay=true)
   &SpawnRate(SpawnRate=8, Use Spawn Probability=false)
  ]
  particleSpawnScript [
   &InitializeParticle(
    Color=<rgba 5.8747 15.2944 30 1>
    Lifetime=&RandomRangeFloat(Minimum=0.4, Maximum=0.6, Evaluation Type=::NewEnumerator1)
    Sprite Size Mode=::NewEnumerator1
    Uniform Sprite Size Max=4
    Uniform Sprite Size Min=1.6
    UsePositionOffset=false
    Write Material Random=true
   )
   &AddVelocity(
    Cone Angle=&RandomRangeFloat(Minimum=0, Maximum=0.3, Evaluation Type=::NewEnumerator1)
    Cone Angle Mode:ENiagara_AngleInput=::NewEnumerator1
    Cone Axis=<0 1 0>
    Module.UseConeFalloff=false
    Module.UseVelDistribution=false
    Override Seed=false
    Velocity Mode=::NewEnumerator2
    Velocity Speed=&RandomRangeFloat(Minimum=1, Maximum=10, Evaluation Type=::NewEnumerator1)
   )
   &DynamicMaterialParameters(
    Index 0 Param 1=&RandomRangeFloat(Minimum=0, Maximum=0.1, Evaluation Type=::NewEnumerator1)
    Index 0 Param 4=&RandomRangeFloat(Minimum=0, Maximum=0.5, Evaluation Type=::NewEnumerator1)
    Module.Index0Float0Write=true
    Module.Index0Float1Write=false
    Module.Index0Float2Write=false
    Module.Index0Float3Write=true
   )
   &ShapeLocation(
    Box Size=<25 60 30>
    MaskSpawnGroup=false
    Offset=<0 6 47>
    Offset Coordinate Space=::Local
    Offset Mode=::NewEnumerator0
    Override Seed=false
    Shape Primitive=::NewEnumerator2
    Surface Expansion Mode:ENiagaraExpansionMode=::NewEnumerator2
    Surface Only=true
    Surface Only Band Thickness=1
    UseOwnerScale=true
   )
   &InheritVelocity(Inherited Velocity Amount Scale=<0.05 0.05 0.05>, Inherited Velocity Speed Limit=0, Speed Limit=false, UseSpeedThreshold=false)
  ]
  particleUpdateScript [
   &ParticleState()
   &ScaleSpriteSize(
    Scale Sprite Size Mode=::NewEnumerator0
    Uniform Scale Factor=&FloatFromCurve(FloatCurve=DataInterface() {propertyValues {Curve {keys [
         {arriveTangent=27.05, interpMode=RCIM_Cubic, leaveTangent=27.05, tangentMode=RCTM_User, time=0, value=0}
         {arriveTangent=1.3101, interpMode=RCIM_Cubic, leaveTangent=1.3101, time=0.0546, value=1.2477}
         {arriveTangent=-1.3446, interpMode=RCIM_Cubic, leaveTangent=-1.3446, time=0.1581, value=0.2072}
         {arriveTangent=0.1974, interpMode=RCIM_Cubic, leaveTangent=0.1974, time=0.2882, value=0.9336}
         {arriveTangent=0.2295, interpMode=RCIM_Cubic, leaveTangent=0.2295, time=0.4301, value=0.2608}
         {arriveTangent=-0.3367, interpMode=RCIM_Cubic, leaveTangent=-0.3367, time=0.5884, value=1.0025}
         {arriveTangent=-1.2837, interpMode=RCIM_Cubic, leaveTangent=-1.2837, time=0.7448, value=0.1549}
         {arriveTangent=-0.607, interpMode=RCIM_Cubic, leaveTangent=-0.607, time=0.834, value=0.6872}
         <at 1 0 RCIM_Cubic>
        ]}}})
   )
   &AccelerationForce(Acceleration=&RandomRangeVector(Minimum=<0 0 -50>, Maximum=<0 0 -80>))
   &Drag(Drag=7, Use Linear Drag=true, Use Rotational Drag=true)
   &SolveForcesAndVelocity(Clamp Velocity=false, Limit Acceleration=false)
   &DynamicMaterialParameters(
    Index 0 Param 2=&FloatFromCurve(FloatCurve=DataInterface() {propertyValues {Curve {keys [<at 0.6 1 RCIM_Cubic>,<at 1 0 RCIM_Cubic>]}}})
    Index 0 Param 3=&FloatFromCurve(FloatCurve=DataInterface() {propertyValues {Curve {keys [<at 0 0>,<at 0.5 20>]}}})
    Module.Index0Float0Write=false
    Module.Index0Float1Write=true
    Module.Index0Float2Write=true
    Module.Index0Float3Write=false
   ) {module=DynamicMaterialParameters001}
   &ScaleColor(
    Scale Alpha=&FloatFromCurve(FloatCurve=DataInterface() {propertyValues {Curve {keys [<at 0 1>,<at 0.6 1>,<at 1 0>]}}})
    ScaleA=true
    ScaleRGB=true
   )
  ]
  rendererClasses [&NiagaraSpriteRendererProperties]
  renderers [{rendererClass=&NiagaraSpriteRendererProperties}]
 }
 Water_Drops {
  emitterUpdateScript [
   &EmitterState(
    Delay First Loop Only=true
    Life Cycle Mode=::NewEnumerator1
    Loop Delay=0.3
    Loop Duration=&RandomRangeFloat(Minimum=0, Maximum=1)
    Recalculate Duration Each Loop=true
    UseLoopDelay=true
   )
   &SpawnRate(
    SpawnRate=&MakeCustomFloatFromBool(
     Bool=&SetBoolByFloatComparison(A=&VectorLength(Vector=$Engine.Owner.Velocity), B=100, Comparison Type=::NewEnumerator6)
     True Float=3
     False Float=0
    )
    Use Spawn Probability=false
   )
  ]
  particleSpawnScript [
   &InitializeParticle(
    Color=<rgba 0.2142 0.4953 1 1>
    Lifetime=0.75
    Sprite Size=<4 9>
    Sprite Size Mode=::NewEnumerator4
    UsePositionOffset=false
    Write Material Random=true
   )
   &SystemLocation(
    MaskSpawnGroup=false
    Offset=&RandomRangeVector(Minimum=<2 -10 48>, Maximum=<2 30 60>, Evaluation Type=::NewEnumerator1)
   )
   &InheritVelocity(Inherited Velocity Amount Scale=<0.7 0.7 -1>, Inherited Velocity Speed Limit=0, Speed Limit=false, UseSpeedThreshold=false)
  ]
  particleUpdateScript [
   &ParticleState()
   &ScaleSpriteSize(
    Scale Sprite Size Mode=::NewEnumerator0
    Uniform Scale Factor=&FloatFromCurve(FloatCurve=DataInterface() {propertyValues {Curve {keys [
         {arriveTangent=-0.0075, time=-0.0009, value=0.8285}
         {arriveTangent=-0.8277, interpMode=RCIM_Cubic, leaveTangent=-0.8277, time=0.6547, value=0.821}
         <at 1 0>
        ]}}})
   )
   &AccelerationForce(Acceleration=<0 0 -450>)
   &Drag(Use Linear Drag=true, Use Rotational Drag=true)
   &SolveForcesAndVelocity(Clamp Velocity=false, Limit Acceleration=false)
   &ScaleColor(
    Scale Alpha=&FloatFromCurve(FloatCurve=DataInterface() {propertyValues {Curve {keys [
         {arriveTangent=0.1936, interpMode=RCIM_Cubic, leaveTangent=-0.3232, tangentMode=RCTM_Break, time=0.0812, value=0.001}
         {arriveTangent=1.2507, interpMode=RCIM_Cubic, leaveTangent=1.2507, time=0.4137, value=0.044}
         {arriveTangent=1.3114, interpMode=RCIM_Cubic, leaveTangent=-0.656, tangentMode=RCTM_Break, time=0.4888, value=0.5108}
         {arriveTangent=-1.169, interpMode=RCIM_Cubic, leaveTangent=-1.169, tangentMode=RCTM_Break, time=0.5709, value=0.6039}
         <at 1 0>
        ]}}})
    ScaleA=true
    ScaleRGB=true
   )
  ]
  rendererClasses [&NiagaraSpriteRendererProperties]
  renderers [{rendererClass=&NiagaraSpriteRendererProperties}]
 }
}
summary {
 systemName=NS_Pickaxe_Yeti_Idle
 emitters [
  {emitterName=Water_Drops, rendererClasses [&NiagaraSpriteRendererProperties]}
  {emitterName=Stars, rendererClasses [&NiagaraSpriteRendererProperties]}
 ]
 userVariables []
}
