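# Physics asset for the Yeti pickaxe cosmetic melee weapon, defining physics bodies and joints for its dangling fish and leaf bones.
# A sphere body on the two-handed handle bone anchors five capsule bodies, two for fish and three for leaves.
# Each dangling body is constrained directly to the handle rather than chained, with its own swing limits.
# The fish joints allow some twist, while all three leaf joints have a zero twist limit.
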PhysicsAsset /BRCosmetics/Weapons/FORT_Melee/Pickaxe_Yeti/Meshes/Pickaxe_Yeti_PhysicsAsset
bodies {
 Melee_TwoHanded_Handle [{shapeType=Sphere, radius=2.9326, center=<1.6117 -2.306 38.8498>}]
 dyn_fish_01 [{shapeType=Capsule, length=7.1875, radius=5.7578, center=<8.5849 -0.3916 -1.3355>, rotation=<rot 0 96.0623 -90.4348>}]
 dyn_fish_02 [{shapeType=Capsule, length=8.395, radius=5.9569, center=<-0.1693 0.175 9.8054>, rotation=<rot -0.569 -89.9999 -1.8844>}]
 dyn_leaf_01 [{shapeType=Capsule, length=5.2334, radius=2.9369, center=<4.4049 1.0366 -2.908>, rotation=<rot -18.6378 -90 -36.3472>}]
 dyn_leaf_02 [{shapeType=Capsule, length=5.3579, radius=2.7596, center=<-3.3032 0.2523 -3.2386>, rotation=<rot 0.7136 -90 30.7479>}]
 dyn_leaf_03 [{shapeType=Capsule, length=5.1837, radius=2.5026, center=<4.6548 -0.4461 -1.7372>, rotation=<rot 0 -103.9785 -49.2965>}]
}
constraints [
 {bone1Name=dyn_fish_01, bone2Name=Melee_TwoHanded_Handle, swing1LimitDegrees=54, swing2LimitDegrees=5, twistLimitDegrees=4}
 {bone1Name=dyn_fish_02, bone2Name=Melee_TwoHanded_Handle, swing1LimitDegrees=21, swing2LimitDegrees=50, twistLimitDegrees=53.1429}
 {bone1Name=dyn_leaf_01, bone2Name=Melee_TwoHanded_Handle, swing1LimitDegrees=45, swing2LimitDegrees=45, twistLimitDegrees=0}
 {bone1Name=dyn_leaf_02, bone2Name=Melee_TwoHanded_Handle, swing1LimitDegrees=10.7143, swing2LimitDegrees=38.1429, twistLimitDegrees=0}
 {bone1Name=dyn_leaf_03, bone2Name=Melee_TwoHanded_Handle, swing1LimitDegrees=45, swing2LimitDegrees=45, twistLimitDegrees=0}
]
